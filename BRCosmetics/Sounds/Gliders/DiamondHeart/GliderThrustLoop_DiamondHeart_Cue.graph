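# Looping thrust sound cue for the DiamondHeart glider cosmetic, using player glider thrust attenuation.
# A mixer layers four looping waves (wind, flap, high and low), with the high layer reduced in the mix.
# Each layer passes through its own continuous modulator, where the AnyDirection parameter scales volume and slightly shifts pitch, so the blend follows that input at runtime.
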SoundCue /BRCosmetics/Sounds/Gliders/DiamondHeart/GliderThrustLoop_DiamondHeart_Cue
root=Mixer_0
nodes [
 {
  id=Mixer_0
  children [ModulatorContinuous_0, ModulatorContinuous_2, ModulatorContinuous_3, ModulatorContinuous_4]
  properties {inputVolume [1,1,0.7,1]}
 }
 {
  id=ModulatorContinuous_0
  children [WavePlayer_0]
  properties {
   pitchModulationParams {default=0, maxOutput=1.1, minOutput=1, parameterName=AnyDirection}
   volumeModulationParams {default=0, maxOutput=0.9, minOutput=0.3, parameterName=AnyDirection}
  }
 }
 {
  id=ModulatorContinuous_2
  children [WavePlayer_1]
  properties {
   pitchModulationParams {default=0, minOutput=1, parameterName=AnyDirection}
   volumeModulationParams {default=0, maxOutput=0.8, minOutput=0.3, parameterName=AnyDirection}
  }
 }
 {
  id=ModulatorContinuous_3
  children [WavePlayer_2]
  properties {
   pitchModulationParams {default=0, minOutput=0.99, parameterName=AnyDirection}
   volumeModulationParams {default=0, maxOutput=0.6, parameterName=AnyDirection}
  }
 }
 {
  id=ModulatorContinuous_4
  children [WavePlayer_3]
  properties {
   pitchModulationParams {default=0, maxOutput=1.1, minOutput=1, parameterName=AnyDirection}
   volumeModulationParams {default=0, minOutput=0.4, parameterName=AnyDirection}
  }
 }
 {id=WavePlayer_0, properties {bLooping=true, soundWaveAssetPtr=&/Glider_DiamondHeart_Thrust_LP_Wind}}
 {id=WavePlayer_1, properties {bLooping=true, soundWaveAssetPtr=&/Glider_DiamondHeart_Thrust_LP_Flap}}
 {id=WavePlayer_2, properties {bLooping=true, soundWaveAssetPtr=&/Glider_DiamondHeart_Thrust_LP_High}}
 {id=WavePlayer_3, properties {bLooping=true, soundWaveAssetPtr=&/Glider_DiamondHeart_Thrust_LP_Low}}
]
properties {
 attenuationSettings=&Fort_Player_Glider_Thrust_Athena_Att
 audioPropertiesSheet=&APS_Cosmetic_Glider_Loop
 maxDistance=12500
 soundClassObject=&Fort_Gliders_Land
 volumeMultiplier=0.9
 concurrencySet [null]
}
